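# PistolCanary_tac7000_aud is a Level Sequence that serves as the audio layer for the tac7000 shot of the PistolCanary tentacle boss sequence.
# A root-level TentacleAudio track holds two overlapping sections that end together: a short late one and a long one starting before frame zero.
# Both keep default volume and pitch, with no sound asset or attach actor shown.
# The TentacleBoss possessable binding carries no tracks.
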LevelSequence /PistolCanary/Sequence/tac/tac7000/PistolCanary_tac7000_aud
bindings [TentacleBoss {id=#1, possessed_class=PistolCanary_TentacleBoss_C, tracks []}]
display_rate {denominator=1, numerator=30}
tick_resolution {denominator=1, numerator=24000}
tracks [{
  type=MovieSceneAudioTrack
  display_name=TentacleAudio
  sections [
   {
    type=MovieSceneAudioSection
    end_frame=150
    start_frame=120
    channels [
     Volume {type=MovieSceneScriptingFloatChannel, default=1}
     Pitch {type=MovieSceneScriptingFloatChannel, default=1}
     AttachActor {type=MovieSceneScriptingActorReferenceChannel}
    ]
   }
   {
    type=MovieSceneAudioSection
    end_frame=150
    start_frame=-673
    channels [
     Volume {type=MovieSceneScriptingFloatChannel, default=1}
     Pitch {type=MovieSceneScriptingFloatChannel, default=1}
     AttachActor {type=MovieSceneScriptingActorReferenceChannel}
    ]
   }
  ]
 }]
